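# Post-sim animation Blueprint for the MoonShock tail cosmetic, which adapts the tail's material when the character wears a backbling.
# The AnimGraph passes the input pose straight through to the output.
# When character customization loads, it checks for a backbling, treats a long one as a cape, and if either applies sets the "Hide Element 1" scalar on the part mesh's dynamic material.
# RelaxSpeed and TrailAlpha are declared but unused in these graphs.
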AnimBlueprint /BRCosmetics/Accessories/FORT_Tails/Tail_M_MED_MoonShock/Meshes/MoonShock_Tail_AnimBP_PostSim
target_skeleton=&/M_MED_MoonShock_Tail_Skeleton
graphs [
 AnimGraph {nodes [Output Pose(Result=Input Pose()) {layerGroup=None}]}
 EventGraph {nodes [
   Event OnCharacterCustomizationLoaded()
   n1 = CheckForBackbling()
   > Set isWearingBackbling(isWearingBackbling=@n1)
   > Set isWearingCape(isWearingCape=@n1.bIsBackblingLong)
   > Branch(Condition=OR Boolean(A=Get isWearingCape() -> (isWearingCape:bool), B=Get isWearingBackbling()))
   > n2 = Get Dynamic Material Instance(SkeletalMeshComponent=GetPartSkeletalMeshComponent())
   > SetScalarParameterValue(self=@n2, ParameterName=Hide Element 1, Value=1)
  ]}
]
variables [
 isWearingBackbling {type=boolean}
 isWearingCape {type=boolean}
 RelaxSpeed {type=number}
 TrailAlpha {type=number}
]
